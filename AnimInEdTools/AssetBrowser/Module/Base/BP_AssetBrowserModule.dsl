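; Base Blueprint for an Asset Browser module in the AnimInEdTools editor tooling, supplying the module's editor utility widgets.
; GetModuleWidget returns its cached instance if valid, otherwise blocking-loads the configured widget class, constructs it and stores it.
; GetCreateWidget loads and constructs a fresh create widget on every call, overwriting the stored instance.
; GetAssetNames and SetCurrentAsset are empty stubs, apparently for child modules to override.
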(fn GetAssetNames
  (return 0))

(fn GetCreateWidget
  (bind _returnvalue (Utilities|LoadClassAssetBlocking (Variables|Defaults|GetCreate%20Widget)))
  (bind _returnvalue_2 (Game|ConstructObjectfromClass _returnvalue))
  (bind _aseditor_utility_widget (Utilities|Casting|CastToEditorUtilityWidget _returnvalue_2))
  (bind _output_get (Variables|Default|SetCreate%20Widget%20Instance _aseditor_utility_widget))
  (return _output_get))

(fn GetModuleWidget
  (Utilities|IsValid (Variables|Default|GetModule%20Widget%20Instance)
    (:Is%20Valid
      (return (Variables|Default|GetModule%20Widget%20Instance)))
    (:Is%20Not%20Valid
      (bind _returnvalue (Utilities|LoadClassAssetBlocking (Variables|Defaults|GetModule%20Widget)))
      (bind _returnvalue_2 (Game|ConstructObjectfromClass _returnvalue))
      (bind _aseditor_utility_widget (Utilities|Casting|CastToEditorUtilityWidget _returnvalue_2))
      (bind _output_get (Variables|Default|SetModule%20Widget%20Instance _aseditor_utility_widget))
      (return _output_get))))

(fn SetCurrentAsset
  (params (Asset :type String)))
