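; Asset browser breadcrumb module that sets the breadcrumb asset shown by its widget from an asset name string.
; SetCurrentAsset queries the Asset Registry for assets of the Bread Crumb class, then synchronously loads each and casts it to ABT_Abstract.
; A case-insensitive name match is passed to the module widget through the SetAssetBreadcrumb interface message.
; The loop never breaks on a match, so every candidate gets loaded.
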(fn SetCurrentAsset
  (params (Asset :type String))
  (bind _outassetdata (AssetRegistry|GetAssetsbyClass (AssetRegistry|GetAssetRegistry) (Utilities|GetSoftClassTopLevelAssetPath (Variables|Defaults|GetBread%20Crumb)) false))
  (bind _array_element (Utilities|Array|ForEachLoopwithBreak _outassetdata)
    (:LoopBody
      (bind (_array_element_packagename _array_element_packagepath _array_element_assetname) (Utilities|Struct|BreakAssetData _array_element))
      (bind _returnvalue (Utilities|LoadAssetBlocking (Utilities|ToSoftObjectReference (Utilities|MakeSoftObjectPath (Utilities|String|Append (Utilities|String|ToString(Name) _array_element_packagename) "." (Utilities|String|ToString(Name) _array_element_assetname))))))
      (bind _asabt_abstract (Utilities|Casting|CastToABT_Abstract _returnvalue))
      (if (Utilities|String|Equal,CaseInsensitive(String) (Variables|/AnimInEdTools/AssetBrowser/Blueprint/ABT_Abstract.ABT_Abstract_C|GetName _asabt_abstract) (Variables|GetAsset))
        (Class|BPIAssetBrowserWidget|SetAssetBreadcrumb(Message) (Variables|Default|GetModule%20Widget%20Instance) _asabt_abstract)))))
